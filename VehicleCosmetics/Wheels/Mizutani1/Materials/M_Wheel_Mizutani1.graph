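# Material for the Mizutani1 skeletal wheel, adding configurable rim and tire glow on top of MF_WheelParent_Simple.
# Custom Color drives tint and intensity through a brightness estimate built from green twice plus blue, ignoring red.
# Switches select flipbook or static rim masks and an optional tire mask, split by red vertex color.
# The blended emissive replaces the parent's, then passes through vehicle damage and de-rez transition functions.
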Material /VehicleCosmetics/Wheels/Mizutani1/Materials/M_Wheel_Mizutani1
graph {
 properties {
  bAutomaticallySetUsageInEditor=false
  bUseMaterialAttributes=true
  bUsedWithMeshDeformer=true
  bUsedWithSkeletalMesh=true
  pixelDepthOffsetMode=PDOM_Legacy
  previewMesh=/VehicleCosmetics/Wheels/Mizutani1/New/Mesh/SK_Wheel_Mizutani1
  usedShadingModels=Default Lit
 }
 nodes [
  Constant()
  n1 = MaterialFunctionCall() {materialFunction=&MF_WheelParent_Simple}
  n2 = GetMaterialAttributes(None=@n1.MaterialAttributes) {attributeGetTypes [#1]}
  n3 = LinearInterpolate(A=@n2.Emissive Color, B=NamedRerouteUsage(), Alpha=OneMinus(None=@n1.VertexColorR))
  MaterialFunctionCall(In=MaterialFunctionCall(Material=SetMaterialAttributes(MaterialAttributes=@n2.MaterialAttributes, Emissive Color=@n3) {attributeSetTypes [#1]}) {materialFunction=&MF_BaseVehicle_Damage}.Result) {materialFunction=&MF_VehicleDeRez_Transition}
  n4 = VectorParameter() {group=Effect, parameterName=Custom Color, sortPriority=-1, defaultValue=<rgba 0 1 0.1559 1>}
  n5 = Add(A=Add(A=@n4.G, B=@n4.G), B=@n4.B)
  n6 = Clamp(None=Multiply(A=@n5, B=Constant() {r=0.3333}))
  n7 = Constant() {r=1.5}
  n8 = Constant() {r=0.35}
  n9 = Constant() {r=1.25}
  n10 = Multiply(A=@n4.RGB, B=Constant() {r=2.5})
  n11 = Constant() {r=0.5}
  n12 = LinearInterpolate(
   A=If(A=@n6, B=Constant() {r=0.111}, "A > B"=Multiply(A=@n7, B=@n4.RGB), "A == B"=Multiply(A=@n8), "A < B"=Multiply(A=@n6, B=@n9))
   B=If(A=@n6, B=Constant() {r=1}, "A > B"=@n10, "A == B"=@n10, "A < B"=Multiply(A=@n4.RGB, B=@n11))
   Alpha=@n6
  )
  n13 = Multiply(A=@n5, B=ScalarParameter() {defaultValue=3, group=Effect, parameterName=Glow Brightness, sortPriority=2})
  n14 = Time()
  n15 = ScalarParameter() {defaultValue=0.5, group=Effect, parameterName=Speed, sortPriority=2}
  n16 = TextureObjectParameter() {
   group=Effect
   parameterName=Rim_Mizutani_RGB
   samplerType=SAMPLERTYPE_LinearColor
   sortPriority=1
   texture=/VehicleCosmetics/Wheels/Mizutani1/Textures/Wheel_Mizutani1_RGB_E
  }
  n17 = MaterialFunctionCall(
   "Animation  Phase (0-1)"=Frac(None=Multiply(A=@n14, B=@n15))
   Number of Rows=ScalarParameter() {defaultValue=4, group=Effect, parameterName=Rows, sortPriority=2}
   Number of Columns=ScalarParameter() {defaultValue=4, group=Effect, parameterName=Columns, sortPriority=2}
   Texture=@n16
   MipBias/Level=ScalarParameter() {defaultValue=-1, group=Effect, parameterName=MipBias, sortPriority=2}
  ) {materialFunction=&FlipBook}
  n18 = NamedRerouteUsage()
  n19 = Constant()
  n20 = StaticSwitchParameter(
   True=StaticSwitchParameter(
    True=StaticComponentMaskParameter(None=@n17.Result) {defaultR=true, group=Effect, parameterName=Rim_Emissive_Mask, sortPriority=0}
    False=ComponentMask(None=@n18) {r=true}
   ) {defaultValue=true, group=Effect, parameterName=Animated_Emissive, sortPriority=0}
   False=@n19
  ) {defaultValue=true, group=Effect, parameterName=Rim Ligts, sortPriority=0}
  n21 = TextureSampleParameter2D() {
   group=Effect
   parameterName=Tire_Emissive
   samplerType=SAMPLERTYPE_LinearColor
   sortPriority=1
   texture=/VehicleCosmetics/Wheels/Mizutani1/Textures/T_Mizutani1_Tire_M
  }
  n22 = StaticSwitchParameter(True=@n21.R, False=@n19) {group=Effect, parameterName=Tire Ligts, sortPriority=0}
  n23 = VertexColor()
  NamedRerouteDeclaration(None=Multiply(A=Multiply(A=@n12, B=@n13), B=LinearInterpolate(A=@n20, B=@n22, Alpha=@n23.R))) {name=Emissive, nodeColor=<rgba 1 0.75 0 1>}
  NamedRerouteDeclaration(None=@n1.RimRGBAMask) {name=Rim_MASKS, nodeColor=<rgba 0 1 0.9297 1>}
 ]
}
